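# Masked, unlit, two-sided ribbon material for the Thrive glider's FX on Niagara mesh particles and ribbons.
# Two panning noises distort UVs and a shape texture is sampled; color mixes face colors, gradients and hue shifts.
# World position offset combines sine/cosine swing, spline thickening and a follow rotation driven by Niagara dynamic parameters, which also control timer, noise and growth.
# ChangeToRibbonParticleMode switches to simpler paths; opacity supports custom-depth and view-angle fades.
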Material /Game/Gadgets/Assets/VinderTech_GliderChute/Glider_Thrive_Female/FX/Material/RibbonAndSpline/M_Thrive_Ribbon_B
graph {
 properties {
  bUsedWithNiagaraMeshParticles=true
  bUsedWithNiagaraRibbons=true
  blendMode=BLEND_Masked
  pixelDepthOffsetMode=PDOM_Legacy
  shadingModel=MSM_Unlit
  twoSided=true
  usedShadingModels=Unlit
 }
 nodes [
  Constant() {r=1}
  Constant()
  DynamicParameter() {parameterIndex=3, defaultValue=<rgba 0 0 0 0>, paramNames [Timer, MeshScale, "NoiseU(SetPar)", "NoiseV(SetPar)"]}
  DynamicParameter() {paramNames ["PosX(OnlyOnSplineMesh)", "PosY(OnlyOnSplineMesh)", "PosZ(OnlyOnSplineMesh)", "RestX(OnlyOnSplineMesh)"]}
  DynamicParameter() {
   parameterIndex=1
   paramNames ["RestY(OnlyOnSplineMesh)", "RestZ(OnlyOnSplineMesh)", "MeshPosX(OnlyOnSplineMesh)", "MeshPosY(OnlyOnSplineMesh)"]
  }
  DynamicParameter() {
   parameterIndex=2
   defaultValue=<rgba 1 0 0 1>
   paramNames ["MeshPosZ(OnlyOnSplineMesh)", "EnableFollow(OnlyOnSplineMesh)", NoiseReducer, Grow]
  }
  n1 = MaterialFunctionCall(Float4=NamedRerouteUsage()) {materialFunction=&BreakOutFloat4Components}
  n2 = Multiply(A=@n1.G) {constB=4}
  n3 = Constant3Vector() {constant=<rgba 0 1 0 0>}
  n4 = LinearInterpolate(A=@n3, B=Constant3Vector() {constant=<rgba 0 0 1 0>}, Alpha=Clamp(None=@n2))
  n5 = Subtract(A=@n2)
  n6 = Clamp(None=@n5)
  n7 = Constant3Vector() {constant=<rgba 0 -1 0 0>}
  n8 = Subtract(A=@n5)
  LinearInterpolate(
   A=LinearInterpolate(A=LinearInterpolate(A=@n4, B=@n7, Alpha=@n6), B=Constant3Vector() {constant=<rgba 0 0 -1 0>}, Alpha=Clamp(None=@n8))
   B=@n3
   Alpha=Clamp(None=Subtract(A=@n8))
  )
  n9 = NamedRerouteUsage()
  n10 = Ceil(None=Subtract(A=ParticleColor().A) {constB=0.3333})
  LinearInterpolate(A=@n9, B=Multiply(A=@n9, B=Saturate(None=@n10)), Alpha=ScalarParameter() {parameterName=20220121-WireFixing})
  MaterialFunctionCall(Emissive=Multiply(A=ParticleColor().RGB, B=NamedRerouteUsage())) {materialFunction=&MF_FX_EmissiveAdjustments}
  n11 = NamedRerouteUsage()
  n12 = NamedRerouteUsage()
  n13 = Add(A=NamedRerouteUsage(), B=@n11)
  Multiply(
   A=StaticSwitchParameter(True=@n11, False=Add(A=@n12, B=@n13)) {group=********, parameterName=ChangeToRibbonParticleMode}
   B=ScalarParameter() {defaultValue=1, group=**Debug, parameterName=TotalWPOScale}
  )
  n14 = TextureCoordinate()
  n15 = MaterialFunctionCall(Float2=@n14) {materialFunction=&BreakOutFloat2Components}
  n16 = OneMinus(None=@n15.R)
  n17 = ScalarParameter() {group=01.GlobalUVSettings, parameterName=FlipU}
  n18 = OneMinus(None=@n15.G)
  n19 = ScalarParameter() {group=01.GlobalUVSettings, parameterName=FlipV}
  NamedRerouteDeclaration(None=MaterialFunctionCall(X=LinearInterpolate(A=@n15.R, B=@n16, Alpha=@n17), Y=LinearInterpolate(A=@n15.G, B=@n18, Alpha=@n19)) {materialFunction=&MakeFloat2}.Result) {name=R_UV, nodeColor=<rgba 0 1 0.2266 1>}
  n20 = DynamicParameter() {parameterIndex=3, defaultValue=<rgba 0 0 0 0>, paramNames [Timer, MeshScale, "NoiseU(SetPar)", "NoiseV(SetPar)"]}
  n21 = LinearInterpolate(A=Time() {bOverride_Period=true, period=360}, B=@n20.Timer, Alpha=Saturate(None=Ceil(None=@n20.Timer)))
  n22 = Add(A=@n21, B=Multiply(A=VertexColor().A, B=ScalarParameter() {defaultValue=10, group=02.GlobalTimeSetting, parameterName=RandomScale}))
  NamedRerouteDeclaration(None=StaticSwitchParameter(True=@n21, False=StaticSwitchParameter(True=@n21, False=@n22) {group=********, parameterName=DisableThriveSpecialColorFixing}) {group=********, parameterName=ChangeToRibbonParticleMode}) {name=R_TimeRandom, nodeColor=<rgba 0 0.7891 1 1>}
  n23 = CurveAtlasRowParameter(CurveTime=MaterialFunctionCall(Float2=NamedRerouteUsage()) {materialFunction=&BreakOutFloat2Components}.R) {
   atlas=/Game/Gadgets/Assets/VinderTech_GliderChute/Glider_Thrive_Female/FX/Texture/CA_Thrive
   curve=/Game/Gadgets/Assets/VinderTech_GliderChute/Glider_Thrive_Female/FX/Texture/CL_ThriveRibbonA
   group=04.ParamCurve
   parameterName=Ribbon_ParamCurve
  }
  NamedRerouteDeclaration(None=AppendVector(A=@n23, B=@n23)) {name=R_ParamCurve, nodeColor=<rgba 0.6172 1 0 1>}
  n24 = VectorParameter() {group=40.NiagaraAnimation, parameterName=GrowAnimation_Detail2, defaultValue=<rgba 1 0.75 0 0>}
  n25 = VectorParameter() {group=40.NiagaraAnimation, parameterName=GrowAnimation_Detail, defaultValue=<rgba 1.5 1 -1.5 0>}
  n26 = DynamicParameter() {
   parameterIndex=2
   defaultValue=<rgba 1 0 0 1>
   paramNames ["MeshPosZ(OnlyOnSplineMesh)", "EnableFollow(OnlyOnSplineMesh)", NoiseReducer, Grow]
  }
  n27 = Add(A=@n25.A, B=OneMinus(None=Saturate(None=@n26.Grow)))
  NamedRerouteDeclaration(None=Saturate(None=LinearInterpolate(A=@n24.R, B=@n24.R, Alpha=@n27))) {name=R_NiagaraAnimation_Grow_BaseShapeWidthScale, nodeColor=<rgba 0.7109 1 0 1>}
  n28 = Multiply(
   A=MaterialFunctionCall(Float2=NamedRerouteUsage()) {materialFunction=&BreakOutFloat2Components}.R
   B=@n25.R
  )
  n29 = LinearInterpolate(A=@n25.G, B=@n25.G, Alpha=@n27)
  NamedRerouteDeclaration(None=Saturate(None=Add(A=@n28, B=@n29))) {name=R_NiagaraAnimation_Grow_SplineWidthScale, nodeColor=<rgba 0.7109 1 0 1>}
  n30 = DynamicParameter() {parameterIndex=3, defaultValue=<rgba 0 0 0 0>, paramNames [Timer, MeshScale, "NoiseU(SetPar)", "NoiseV(SetPar)"]}
  n31 = NamedRerouteUsage()
  n32 = Multiply(A=ScalarParameter() {group=32.SplineWPO, parameterName=Spline_WidthBase}, B=@n31)
  n33 = Multiply(A=@n1.R, B=@n32)
  n34 = MaterialFunctionCall!(
   UseRemapCurver=ScalarParameter() {group=32.SplineWPO, parameterName=Spline_SplineThicken_UseCurverForSplineThickenWidth}
   WidthBase=@n32
   WidthRemapCurver=@n33
   TangentUToWorld=NamedRerouteUsage()
   TangentVToWorld=NamedRerouteUsage()
   WidthTip=Multiply(A=ScalarParameter() {group=32.SplineWPO, parameterName=Spline_SplineThicken_WidthTip}, B=@n31)
   WorldPosition=Add(A=NamedRerouteUsage(), B=WorldPosition() {worldPositionShaderOffset=WPT_ExcludeAllShaderOffsets}.XYZ)
   UVs for Projection
   UVs for Thickness
   UVs for texturing
   Expand U or V UV Channel=StaticBool()
   DeriveNormalZ
   AdditionalNormal
   AngleCorrectedNormals
   FlattenPixelNormal
   UVs For NormalShading
  ) {materialFunction=/Game/Gadgets/Assets/VinderTech_GliderChute/Glider_Thrive_Female/FX/Material/MF_Thrive_CustomSplineThicken}
  n35 = Transform(None=Multiply(A=Normalize(VectorInput=@n3), B=ScalarParameter() {defaultValue=1, group=32.SplineWPO, parameterName=Spline_CurveControl_FlipVDir}))
  n36 = NamedRerouteUsage()
  n37 = Multiply(A=Subtract(A=MaterialFunctionCall(Float2=@n36) {materialFunction=&BreakOutFloat2Components}.G) {constB=0.5}) {constB=2}
  n38 = StaticSwitchParameter(True=@n34.WorldPositionOffset, False=Multiply(A=Multiply(A=@n35, B=@n37), B=@n33)) {group=32.SplineWPO, parameterName=UseSplineThickness}
  NamedRerouteDeclaration(None=LinearInterpolate(A=@n38, B=Multiply(A=@n38, B=@n30.MeshScale), Alpha=ScalarParameter() {parameterName=ApplyDynamicInputMeshScale})) {name=R_SplineWPO, nodeColor=<rgba 0.5703 0 1 1>}
  n39 = MaterialFunctionCall(Float2=NamedRerouteUsage()) {materialFunction=&BreakOutFloat2Components}
  n40 = Add(
   A=Multiply(A=OneMinus(None=@n39.R), B=ScalarParameter() {group=03.BaseShape, parameterName=BaseShape_Stretch})
   B=ScalarParameter() {group=03.BaseShape, parameterName=BaseShape_StretchOffset}
  )
  n41 = Constant()
  n42 = Add(A=Add(A=AppendVector(A=@n40, B=@n41), B=NamedRerouteUsage()), B=NamedRerouteUsage())
  NamedRerouteDeclaration(None=@n42) {name=R_BaseShapeUV, nodeColor=<rgba 0 1 0.625 1>}
  n43 = VectorParameter() {group=03.BaseShape, parameterName=BaseShape_UVScale, channelNames {a="", b="", g=VScale, r=UScale}, defaultValue=<rgba 1 1 0 1>}
  n44 = Multiply(A=NamedRerouteUsage(), B=@n43.VScale)
  n45 = TextureSampleParameter2D(UVs=MaterialFunctionCall(UVs=@n42, Texture Scale=AppendVector(A=@n43.UScale, B=@n44)) {materialFunction=&ScaleUVsByCenter}.UVs) {
   group=03.BaseShape
   parameterName=BaseShape_Texture
   samplerSource=SSM_Clamp_WorldGroupSettings
   samplerType=SAMPLERTYPE_Grayscale
   texture=/Game/Gadgets/Assets/VinderTech_GliderChute/Glider_Thrive_Female/FX/Texture/T_RibbonBaseShape
  }
  NamedRerouteDeclaration(None=@n45.R) {name=R_BaseShapeSampledValue, nodeColor=<rgba 0.6466 0.4328 1 1>}
  n46 = MaterialFunctionCall(Float3=LinearInterpolate(
    A=SceneTexture() {sceneTextureId=PPI_CustomDepth}.Color
    B=SceneDepth().R
    Alpha=ScalarParameter() {group=05.CustomDepthCullAndFade, parameterName=Use Scene Depth Instead}
   )) {materialFunction=&BreakOutFloat3Components}
  n47 = PixelDepth()
  n48 = ComponentMask(None=@n47.R) {r=true}
  n49 = Saturate(None=Divide(A=Subtract(A=@n46.R, B=@n48), B=ScalarParameter() {defaultValue=50, group=05.CustomDepthCullAndFade, parameterName=CustomDepthFadeDistance}) {constB=50})
  n50 = Constant() {r=1}
  n51 = If(A=@n47.R, B=@n46.R, "A > B"=Constant(), "A == B"=@n50, "A < B"=@n50) {equalsThreshold=0.1}
  n52 = Constant() {r=1}
  NamedRerouteDeclaration(None=ShadingPathSwitch(Default=Multiply(A=@n49, B=@n51), Forward=@n52, Mobile=@n52)) {name=R_CustomDepthCullAndFade, nodeColor=<rgba 0 0.6953 1 1>}
  n53 = VectorParameter() {group=22.UVBaseColorGradient, parameterName=Gradient_U_Detail, channelNames {a=ColorHueshift, b=ColorMul, g=hardness, r=Uoffset}}
  n54 = Add(
   A=MaterialFunctionCall(Float2=NamedRerouteUsage()) {materialFunction=&BreakOutFloat2Components}.R
   B=@n53.Uoffset
  )
  n55 = Multiply(A=Saturate(None=Multiply(A=@n54, B=@n53.hardness)), B=@n53.ColorMul)
  n56 = Multiply(A=@n55, B=@n53.ColorHueshift)
  n57 = MaterialFunctionCall(Float2=NamedRerouteUsage()) {materialFunction=&BreakOutFloat2Components}
  n58 = OneMinus(None=@n57.G)
  n59 = NamedRerouteUsage()
  n60 = Transform(None=Constant3Vector() {constant=<rgba 1 0 0 0>}) {transformSourceType=TRANSFORMSOURCE_Local}
  n61 = Saturate(None=Multiply(A=DotProduct(A=@n59, B=@n60)) {constB=10})
  n62 = LinearInterpolate(A=@n58, B=@n57.G, Alpha=@n61) {desc=r}
  n63 = LinearInterpolate(A=@n57.G, B=@n58, Alpha=@n61) {desc=l}
  n64 = Ceil(None=Subtract(A=VertexColor().B) {constB=0.01})
  n65 = DotProduct(A=@n59, B=Constant3Vector() {constant=<rgba 0 0 1 0>})
  n66 = Saturate(None=Multiply(A=Subtract(A=@n65) {constB=0.4}) {constB=4})
  n67 = ScalarParameter() {group=22.UVBaseColorGradient, parameterName=Gradient_SpecialFIx}
  n68 = LinearInterpolate(A=@n57.G, B=LinearInterpolate(A=LinearInterpolate(A=@n62, B=@n63, Alpha=@n64), B=@n57.G, Alpha=@n66), Alpha=@n67)
  n69 = VectorParameter() {group=22.UVBaseColorGradient, parameterName=Gradient_V_Detail, channelNames {a=ColorHueshift, b=ColorMul, g=hardness, r=Voffset}}
  n70 = Add(
   A=StaticSwitchParameter(True=@n57.G, False=StaticSwitchParameter(True=@n57.G, False=@n68) {group=********, parameterName=DisableThriveSpecialColorFixing}) {group=********, parameterName=ChangeToRibbonParticleMode}
   B=@n69.Voffset
  )
  n71 = ScalarParameter() {defaultValue=1, group=22.UVBaseColorGradient, parameterName=Gradient_SpecialFixMul}
  n72 = LinearInterpolate(A=@n69.ColorMul, B=LinearInterpolate(A=@n69.ColorMul, B=Multiply(A=@n69.ColorMul, B=@n71), Alpha=@n66), Alpha=@n67)
  n73 = Multiply(
   A=Saturate(None=Multiply(A=@n70, B=@n69.hardness))
   B=StaticSwitchParameter(
    True=@n69.ColorMul
    False=StaticSwitchParameter(True=@n69.ColorMul, False=@n72) {group=********, parameterName=DisableThriveSpecialColorFixing}
   ) {group=********, parameterName=ChangeToRibbonParticleMode}
  )
  n74 = Multiply(A=@n73, B=@n69.ColorHueshift)
  n75 = Add(A=@n73)
  n76 = Multiply(A=Add(A=@n55), B=NamedRerouteUsage())
  NamedRerouteDeclaration(None=MaterialFunctionCall(Hue Shift Percentage=Add(A=@n56, B=@n74), Texture=Multiply(A=@n75, B=@n76)) {materialFunction=&HueShift}.Result) {name=R_Color, nodeColor=<rgba 0 0.1094 1 1>}
  NamedRerouteDeclaration(None=@n34.Normal) {name=R_SplineWPO_Normal, nodeColor=<rgba 0.0312 1 0 1>}
  n77 = Multiply(
   A=ComponentMask(None=NamedRerouteUsage()) {r=true}
   B=ScalarParameter() {defaultValue=1, group=SimpleColorGradient, parameterName=Shift_Gradient}
  )
  n78 = LinearInterpolate(
   A=VectorParameter() {group=SimpleColorGradient, parameterName=ColorA, defaultValue=<rgba 0 0.132 1 1>}.RGB
   B=VectorParameter() {group=SimpleColorGradient, parameterName=ColorB, defaultValue=<rgba 1 0.0693 0 1>}.RGB
   Alpha=Power(Base=Clamp(None=@n77), Exp=ScalarParameter() {defaultValue=2, group=SimpleColorGradient, parameterName=Gradient_Power})
  )
  n79 = VectorParameter() {group=20.BaseColor, parameterName=BaseColor_FrontFace, defaultValue=<rgba 1 0 0.6179 1>}
  n80 = Saturate(None=TwoSidedSign())
  n81 = LinearInterpolate(A=VectorParameter() {group=20.BaseColor, parameterName=BaseColor_BackFace, defaultValue=<rgba 1 0 0.6179 1>}.RGB, B=@n79.RGB, Alpha=@n80)
  n82 = VectorParameter() {group=22.UVBaseColorGradient, parameterName=SpecialGradient_U_Detail, channelNames {a="", b=Colormul, g=hardness, r=Uoffset}}
  n83 = Add(
   A=MaterialFunctionCall(Float2=NamedRerouteUsage()) {materialFunction=&BreakOutFloat2Components}.R
   B=@n82.Uoffset
  )
  n84 = Multiply(A=Saturate(None=Multiply(A=@n83, B=@n82.hardness)), B=@n82.Colormul)
  n85 = VectorParameter() {group=21.AdditionalTexture, parameterName=AdditionalTexture_Color_Frontface, defaultValue=<rgba 0 0 0 1>}
  n86 = VectorParameter() {group=21.AdditionalTexture, parameterName=AdditionalTexture_Color_Backface, defaultValue=<rgba 0 0 0 1>}
  n87 = VectorParameter() {
   group=21.AdditionalTexture
   parameterName=AdditionalTexture_UVScaleAndNoiseDistortStrength
   channelNames {a=NoiseVScale, b=NoiseUScale, g=VScale, r=UScale}
  }
  n88 = MaterialFunctionCall(
   UVs=Add(A=NamedRerouteUsage(), B=AppendVector(A=@n87.NoiseUScale, B=@n87.NoiseUScale))
   Texture Scale=AppendVector(A=@n87.UScale, B=@n87.UScale)
  ) {materialFunction=&ScaleUVsByCenter}
  n89 = VectorParameter() {
   group=21.AdditionalTexture
   parameterName=AdditionalTexture_ValueTweakDetail
   channelNames {a="", b=contrast, g=mul, r=add}
   defaultValue=<rgba 0 1 0 0>
  }
  n90 = Multiply(
   A=Add(
    A=TextureSampleParameter2D(UVs=@n88.UVs) {group=21.AdditionalTexture, parameterName="AdditionalTexture(R)", texture=&T_BlacksRGB}.R
    B=@n89.add
   )
   B=@n89.mul
  )
  n91 = LinearInterpolate(
   A=Multiply(A=StaticSwitchParameter(True=@n79.RGB, False=@n81) {group=********, parameterName=ChangeToRibbonParticleMode}, B=Add(A=@n84))
   B=StaticSwitchParameter(True=@n85.RGB, False=LinearInterpolate(A=@n86.RGB, B=@n85.RGB, Alpha=@n80)) {group=********, parameterName=ChangeToRibbonParticleMode}
   Alpha=MaterialFunctionCall(In=Saturate(None=@n90), Contrast=@n89.contrast) {materialFunction=&CheapContrast}.Result
  )
  n92 = NamedRerouteUsage()
  n93 = Add(
   A=NamedRerouteUsage()
   B=Multiply(
    A=MaterialFunctionCall(XY=@n92) {materialFunction=&Swizzle}.YX
    B=ScalarParameter() {group=20.BaseColor, parameterName=HueShiftNoiseScale}
   )
  )
  n94 = Add(
   A=Multiply(
    A=MaterialFunctionCall(Float2=@n93) {materialFunction=&BreakOutFloat2Components}.R
    B=ScalarParameter() {group=20.BaseColor, parameterName=HueShiftStr}
   )
   B=Multiply(A=NamedRerouteUsage(), B=ScalarParameter() {group=20.BaseColor, parameterName=HueShiftUVPanSpeed})
  )
  NamedRerouteDeclaration(None=LinearInterpolate(
    A=Multiply(A=@n78, B=@n91)
    B=MaterialFunctionCall(Hue Shift Percentage=@n94, Texture=@n91) {materialFunction=&HueShift}.Result
    Alpha=ScalarParameter() {defaultValue=1, group=SimpleColorGradient, parameterName=Use_HueShift}
   )) {name=R_BaseColorWithAdditionalTexture, nodeColor=<rgba 1 0 0.7266 1>}
  n95 = VectorParameter() {
   group=10.UVDistortNoise_A
   parameterName=NoiseA_UVTransform
   channelNames {a=PannerV, b=PannerU, g=ScaleV, r=ScaleU}
   defaultValue=<rgba 1 0.1 0.2 0>
  }
  n96 = Multiply(A=NamedRerouteUsage(), B=AppendVector(A=@n95.ScaleU, B=@n95.ScaleU))
  n97 = VectorParameter() {group=10.UVDistortNoise_A, parameterName=NoiseA_UVOffset, channelNames {a="", b="", g=OffsetV, r=OffsetU}, defaultValue=<rgba 1 0.1 0.2 0>}
  n98 = AppendVector(A=@n97.OffsetU, B=@n97.OffsetU)
  n99 = TextureSampleParameter2D(UVs=Panner(Coordinate=Add(A=@n96, B=@n98), Time=NamedRerouteUsage(), Speed=AppendVector(A=@n95.PannerU, B=@n95.PannerU)) {speedX=0.216}) {
   group=10.UVDistortNoise_A
   parameterName=NoiseA_Texture
   samplerType=SAMPLERTYPE_LinearGrayscale
   texture=/Game/Gadgets/Assets/VinderTech_GliderChute/Glider_Thrive_Female/FX/Texture/T_AntiMaterialCharge_Dissolve_01
  }
  n100 = VectorParameter() {
   group=10.UVDistortNoise_A
   parameterName=NoiseA_Property
   channelNames {a=NoiseStrength_V, b=NoiseStrength_U, g=NoiseStrengthOffset_V, r=NoiseStrengthOffset_U}
   defaultValue=<rgba -0.5 -0.5 0.2 0>
  }
  n101 = Add(A=AppendVector(A=@n99.R, B=@n99.R), B=AppendVector(A=@n100.NoiseStrengthOffset_U, B=@n100.NoiseStrengthOffset_U)) {constB=-0.5}
  n102 = DynamicParameter() {parameterIndex=3, defaultValue=<rgba 0 0 0 0>, paramNames [Timer, MeshScale, "NoiseU(SetPar)", "NoiseV(SetPar)"]}
  n103 = LinearInterpolate(
   A=AppendVector(A=@n100.NoiseStrength_U, B=@n100.NoiseStrength_U)
   B=AppendVector(A=@n102."NoiseU(SetPar)", B=@n102."NoiseU(SetPar)")
   Alpha=ScalarParameter() {group=DynamicParameterNoise, parameterName=Use_NoiseNiagaraPar}
  )
  n104 = MaterialFunctionCall(Float2=NamedRerouteUsage()) {materialFunction=&BreakOutFloat2Components}
  n105 = OneMinus(None=@n104.R)
  n106 = VectorParameter() {
   group=10.UVDistortNoise_A
   parameterName=NoiseA_UMaskInfo
   channelNames {a=Buttomoffset, b=ButtomHardness, g=TopOffset, r=TopHardness}
   defaultValue=<rgba 5 0 25 5>
  }
  n107 = Saturate(None=Add(A=Multiply(A=@n105, B=@n106.TopHardness) {constB=15}, B=@n106.TopOffset))
  n108 = Saturate(None=Add(A=Multiply(A=@n104.R, B=@n106.ButtomHardness) {constB=15}, B=@n106.Buttomoffset))
  NamedRerouteDeclaration(None=Multiply(A=Multiply(A=@n101, B=@n103), B=Multiply(A=@n107, B=@n108))) {name=R_NoiseA, nodeColor=<rgba 0 1 0.25 1>}
  n109 = MaterialFunctionCall(Float2=NamedRerouteUsage()) {materialFunction=&BreakOutFloat2Components}
  n110 = OneMinus(None=@n109.R)
  n111 = NamedRerouteUsage()
  n112 = Multiply(A=Saturate(None=Multiply(A=@n110) {constB=2}), B=Multiply(A=@n111, B=@n109.R))
  n113 = ScalarParameter() {group=50.Opacity, parameterName=DisableSimpleOpacityScale}
  n114 = Multiply(
   A=StaticSwitchParameter(True=@n111, False=LinearInterpolate(A=@n112, B=@n111, Alpha=@n113)) {group=********, parameterName=ChangeToRibbonParticleMode}
   B=Multiply(A=ScalarParameter() {group=50.Opacity, parameterName=OpacityScale}, B=ParticleColor().A)
  )
  n115 = StaticSwitchParameter(True=Multiply(A=NamedRerouteUsage(), B=@n114), False=@n114) {group=50.Opacity, parameterName=UseCustomDepthCompare}
  n116 = MaterialFunctionCall(Float4=NamedRerouteUsage()) {materialFunction=&BreakOutFloat4Components}
  n117 = StaticSwitchParameter(
   True=@n115
   False=LinearInterpolate(A=@n115, B=Multiply(A=@n116.B, B=@n115), Alpha=ScalarParameter() {group=50.Opacity, parameterName=CurveBOpacity})
  ) {group=********, parameterName=ChangeToRibbonParticleMode}
  n118 = MaterialFunctionCall("FadeSensitivity (optional)"=ScalarParameter() {defaultValue=2, parameterName=ViewAngleSensitivity}) {materialFunction=&FadeBasedOnViewAngle}
  NamedRerouteDeclaration(None=LinearInterpolate(A=@n117, B=Multiply(A=@n117, B=@n118.Result), Alpha=ScalarParameter() {parameterName=UseViewAngle})) {name=R_Opacity, nodeColor=<rgba 0 1 0.6016 1>}
  NamedRerouteDeclaration(None=@n14) {name=R_OriginalUV, nodeColor=<rgba 0 1 0.4869 1>}
  n119 = VectorParameter() {
   group=10.UVDistortNoise_B
   parameterName=NoiseB_UVTransform
   channelNames {a=PannerV, b=PannerU, g=ScaleV, r=ScaleU}
   defaultValue=<rgba 1 0.1 0.2 0>
  }
  n120 = Multiply(A=NamedRerouteUsage(), B=AppendVector(A=@n119.ScaleU, B=@n119.ScaleU))
  n121 = VectorParameter() {group=10.UVDistortNoise_B, parameterName=NoiseB_UVOffset, channelNames {a="", b="", g=OffsetV, r=OffsetU}, defaultValue=<rgba 1 0.1 0.2 0>}
  n122 = AppendVector(A=@n121.OffsetU, B=@n121.OffsetU)
  n123 = TextureSampleParameter2D(UVs=Panner(Coordinate=Add(A=@n120, B=@n122), Time=NamedRerouteUsage(), Speed=AppendVector(A=@n119.PannerU, B=@n119.PannerU)) {speedX=0.216}) {
   group=10.UVDistortNoise_B
   parameterName=NoiseB_Texture
   samplerType=SAMPLERTYPE_LinearGrayscale
   texture=&T_AntiMaterialCharge_Dissolve_01
  }
  n124 = VectorParameter() {
   group=10.UVDistortNoise_B
   parameterName=NoiseB_Property
   channelNames {a=NoiseStrength_V, b=NoiseStrength_U, g=NoiseStrengthOffset_V, r=NoiseStrengthOffset_U}
   defaultValue=<rgba -0.5 -0.5 0.2 0>
  }
  n125 = Add(A=AppendVector(A=@n123.R, B=@n123.R), B=AppendVector(A=@n124.NoiseStrengthOffset_U, B=@n124.NoiseStrengthOffset_U)) {constB=-0.5}
  n126 = AppendVector(A=@n124.NoiseStrength_U, B=@n124.NoiseStrength_U)
  n127 = MaterialFunctionCall(Float2=NamedRerouteUsage()) {materialFunction=&BreakOutFloat2Components}
  n128 = OneMinus(None=@n127.R)
  n129 = VectorParameter() {
   group=10.UVDistortNoise_B
   parameterName=NoiseB_UMaskInfo
   channelNames {a=Buttomoffset, b=ButtomHardness, g=TopOffset, r=TopHardness}
   defaultValue=<rgba 3 0 5 2.4>
  }
  n130 = Saturate(None=Add(A=Multiply(A=@n128, B=@n129.TopHardness) {constB=15}, B=@n129.TopOffset))
  n131 = Saturate(None=Add(A=Multiply(A=@n127.R, B=@n129.ButtomHardness) {constB=15}, B=@n129.Buttomoffset))
  NamedRerouteDeclaration(None=Multiply(A=Multiply(A=@n125, B=@n126), B=Multiply(A=@n130, B=@n131))) {name=R_NoiseB, nodeColor=<rgba 1 0.7969 0 1>}
  n132 = DynamicParameter() {
   parameterIndex=2
   defaultValue=<rgba 1 0 0 1>
   paramNames ["MeshPosZ(OnlyOnSplineMesh)", "EnableFollow(OnlyOnSplineMesh)", NoiseReducer, Grow]
  }
  n133 = Saturate(None=@n132.NoiseReducer)
  n134 = NamedRerouteUsage()
  n135 = NamedRerouteUsage()
  NamedRerouteDeclaration(None=Multiply(A=OneMinus(None=@n133), B=Add(A=@n134, B=@n135))) {name=R_MixedNoise, nodeColor=<rgba 0 1 0.0859 1>}
  n136 = DynamicParameter() {parameterIndex=3, defaultValue=<rgba 0 0 0 0>, paramNames [Timer, MeshScale, "NoiseU(SetPar)", "NoiseV(SetPar)"]}
  n137 = VectorParameter() {group=30.SwingWPO, parameterName=Swing_Cosine_Mask, channelNames {a="", b="", g=offset, r=Multiply}, defaultValue=<rgba 2 -1.1 0 1>}
  n138 = MaterialFunctionCall(Float2=NamedRerouteUsage()) {materialFunction=&BreakOutFloat2Components}
  n139 = Add(A=Multiply(A=@n137.Multiply, B=@n138.R), B=@n137.offset)
  n140 = VectorParameter() {group=30.SwingWPO, parameterName=Swing_Cosine_Property, channelNames {a="", b=Frequency, g=Strength, r=Speed}, defaultValue=<rgba 1.5 5 2 1>}
  n141 = NamedRerouteUsage()
  n142 = Add(A=Multiply(A=@n138.R, B=@n140.Frequency), B=Multiply(A=@n141, B=@n140.Speed))
  n143 = Cosine(None=@n142)
  n144 = VectorParameter() {group=30.SwingWPO, parameterName=Swing_Cosine_Axis, defaultValue=<rgba 0 1 0 0>}
  n145 = Multiply(A=OneMinus(None=Saturate(None=@n139)), B=Multiply(A=Multiply(A=@n143, B=@n140.Strength), B=Transform(None=@n144.RGB)))
  n146 = VectorParameter() {group=30.SwingWPO, parameterName=Swing_Sine_Mask, channelNames {a="", b="", g=Offset, r=Multiply}, defaultValue=<rgba 2 -0.375 0 1>}
  n147 = Multiply(A=@n146.Multiply, B=@n138.R)
  n148 = OneMinus(None=Saturate(None=Add(A=@n147, B=@n146.Offset)))
  n149 = VectorParameter() {group=30.SwingWPO, parameterName=Swing_Sine_Property, channelNames {a="", b=Frequency, g=Strength, r=Speed}, defaultValue=<rgba 1.5 20 2 1>}
  n150 = Add(A=Multiply(A=@n138.R, B=@n149.Frequency), B=Multiply(A=@n141, B=@n149.Speed))
  n151 = Multiply(
   A=Multiply(A=Sine(None=@n150), B=@n149.Strength)
   B=Transform(None=VectorParameter() {group=30.SwingWPO, parameterName=Swing_Sine_Axis, defaultValue=<rgba 0 0 1 1>}.RGB)
  )
  n152 = StaticSwitchParameter(True=@n145, False=Add(A=Multiply(A=@n148, B=@n151), B=@n145)) {group=********, parameterName=ChangeToRibbonParticleMode}
  NamedRerouteDeclaration(None=LinearInterpolate(A=@n152, B=Multiply(A=@n152, B=@n136.MeshScale), Alpha=ScalarParameter() {parameterName=ApplyDynamicInputMeshScale})) {name=R_SwingWPO, nodeColor=<rgba 1 0.7734 0 1>}
  n153 = DynamicParameter() {paramNames ["PosX(OnlyOnSplineMesh)", "PosY(OnlyOnSplineMesh)", "PosZ(OnlyOnSplineMesh)", "RestX(OnlyOnSplineMesh)"]}
  n154 = DynamicParameter() {
   parameterIndex=1
   paramNames ["RestY(OnlyOnSplineMesh)", "RestZ(OnlyOnSplineMesh)", "MeshPosX(OnlyOnSplineMesh)", "MeshPosY(OnlyOnSplineMesh)"]
  }
  n155 = Normalize(VectorInput=AppendVector(A=AppendVector(A=@n153."RestX(OnlyOnSplineMesh)", B=@n154."RestY(OnlyOnSplineMesh)"), B=@n154."RestZ(OnlyOnSplineMesh)"))
  n156 = AppendVector(A=AppendVector(A=@n153."PosX(OnlyOnSplineMesh)", B=@n153."PosX(OnlyOnSplineMesh)"), B=@n153."PosZ(OnlyOnSplineMesh)")
  n157 = Constant3Vector() {constant=<rgba 0 0 1 0>}
  n158 = DynamicParameter() {
   parameterIndex=2
   defaultValue=<rgba 1 0 0 1>
   paramNames ["MeshPosZ(OnlyOnSplineMesh)", "EnableFollow(OnlyOnSplineMesh)", NoiseReducer, Grow]
  }
  n159 = AppendVector(A=AppendVector(A=@n154."MeshPosX(OnlyOnSplineMesh)", B=@n154."MeshPosX(OnlyOnSplineMesh)"), B=@n158."MeshPosZ(OnlyOnSplineMesh)")
  n160 = Normalize(VectorInput=Subtract(A=Add(A=@n156, B=@n157), B=@n159))
  n161 = CrossProduct(A=@n155, B=@n160)
  n162 = DotProduct(A=@n160, B=@n155)
  n163 = MaterialFunctionCall(Float2=NamedRerouteUsage()) {materialFunction=&BreakOutFloat2Components}
  n164 = ScalarParameter() {group="31.FollowWPO(NiagaraDeform)", parameterName=FollowWPO_Offset}
  n165 = Multiply(A=Add(A=OneMinus(None=@n163.R), B=@n164), B=Divide(B=Add(A=@n164)))
  n166 = Multiply(
   A=@n165
   B=Multiply(
    A=ArccosineFast!(None=@n162)
    B=ScalarParameter() {defaultValue=0.1592, group="31.FollowWPO(NiagaraDeform)", parameterName=FollowWPO_Rigidity}
   )
  )
  n167 = Add(A=NamedRerouteUsage(), B=WorldPosition() {worldPositionShaderOffset=WPT_ExcludeAllShaderOffsets}.XYZ)
  n168 = DynamicParameter() {
   parameterIndex=2
   defaultValue=<rgba 1 0 0 1>
   paramNames ["MeshPosZ(OnlyOnSplineMesh)", "EnableFollow(OnlyOnSplineMesh)", NoiseReducer, Grow]
  }
  NamedRerouteDeclaration(None=LinearInterpolate(
    B=RotateAboutAxis(NormalizedRotationAxis=@n161, RotationAngle=@n166, PivotPoint=@n159, Position=@n167)
    Alpha=@n168."EnableFollow(OnlyOnSplineMesh)"
   )) {name=R_NiagaraDeformationWPO, nodeColor=<rgba 0 1 0.1797 1>}
  n169 = Transform(None=Constant3Vector() {constant=<rgba 1 0 0 0>})
  NamedRerouteDeclaration(None=LinearInterpolate(
    A=@n169
    B=MaterialFunctionCall!(VectorToRotate=@n169, LookAtVector=@n160, RestingVector=@n155) {materialFunction=/Engine/Functions/Engine_MaterialFunctions01/Vectors/RotateVector}.Result
    Alpha=@n165
   )) {name=R_NiagaraDeformation_UDirWS, nodeColor=<rgba 0 1 0.4844 1>}
  n170 = Transform(None=Constant3Vector() {constant=<rgba 1 1 0 0>})
  NamedRerouteDeclaration(None=LinearInterpolate(
    A=@n170
    B=MaterialFunctionCall(VectorToRotate=@n170, LookAtVector=@n160, RestingVector=@n155) {materialFunction=&RotateVector}.Result
    Alpha=@n165
   )) {name=R_NiagaraDeformation_VDirWS, nodeColor=<rgba 0 1 0.4844 1>}
  NamedRerouteUsage()
  PixelNormalWS()
  Sine(None=@n142)
  VertexColor()
 ]
}
